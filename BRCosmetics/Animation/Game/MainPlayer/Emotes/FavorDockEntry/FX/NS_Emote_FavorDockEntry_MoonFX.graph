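# Moon-themed Niagara effect for the FavorDockEntry emote, combining moon lights, shooting stars and smoke.
# MoonLight pins four lights along a ring at the gadget's JNT_MoonSocket, spaced by an array.
# Shooting star heads shift spawn side with camera direction and feed ribbon trails through particle reads.
# Time-of-day-tinted smoke blends from steady spawning to bursts as owner speed rises; MeshStars, NebulaBG and SparkleStars show no renderers.
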NiagaraSystem /BRCosmetics/Animation/Game/MainPlayer/Emotes/FavorDockEntry/FX/NS_Emote_FavorDockEntry_MoonFX
dependencies {
 usedDataInterfaces [
  &NiagaraDataInterfaceCurve
  &NiagaraDataInterfaceCamera
  &NiagaraDataInterfaceColorCurve
  &NiagaraDataInterfaceParticleRead
  &NiagaraDataInterfaceSkeletalMesh
  &NiagaraDataInterfaceArrayFloat
 ]
 usedDynamicInputs [&MaskFloatBySpawnGroup]
 usedModules [
  &self:SystemSpawnScript.SystemScriptSource.SystemScriptGraph.NiagaraNodeAssignment_1.SetVariables_#1
  &SystemState
  &self:SystemSpawnScript.SystemScriptSource.SystemScriptGraph.NiagaraNodeAssignment_2.SetVariables_#2
 ]
}
emitters {
 MeshStars {rendererClasses [], renderers []}
 MoonLight {
  emitterUpdateScript [
   &EmitterState(Life Cycle Mode=::NewEnumerator1, Loop Behavior=::NewEnumerator1, Loop Duration Mode=::NewEnumerator1, UseLoopDelay=false)
   &SpawnBurst_Instantaneous(Spawn Count=4, Use Spawn Probability=false, UseLoopCountLimit=false)
  ]
  particleSpawnScript [
   &InitializeParticle(
    Color=<rgba 2.3578 4.1686 5 1>
    Lifetime Max=0.8
    Lifetime Min=0.6
    Lifetime Mode=::NewEnumerator1
    Mesh Scale Mode=::NewEnumerator3
    Sprite Size Mode=::NewEnumerator1
    Uniform Sprite Size Max=25
    Uniform Sprite Size Min=18
    UsePositionOffset=true
    Write Material Random=true
   )
   &InitialMeshOrientation(Mesh Orientation Mode=::NewEnumerator1, Rotation=<0.25 0.25 -0.16>, Use Rotation Vector=true)
  ]
  particleUpdateScript [
   &ParticleState(Kill Particles When Lifetime Has Elapsed=false)
   &SkeletalMeshLocation(
    MaskSpawnGroup=false
    Mesh Sampling Type=::NewEnumerator5
    Orientation Sampling=::NewEnumerator1
    Position Sampling=::NewEnumerator2
    Skeletal Mesh=DataInterface() {propertyValues {
      DefaultMesh=null
      ExcludeBoneName=root
      OptionalComponentClass=null
      PreviewMesh=/BRCosmetics/Animation/Props/Gadget_FavorDockEntry/Meshes/FavorDockEntry_Gadgets
      SkinningMode=SkinOnTheFly
      SoftSourceActor=null
      SourceMode=Default
      UvSetIndex=0
      WholeMeshLOD=-1
      bExcludeBone=false
      bReadDeformedGeometry=true
      bRequireCurrentFrameData=false
      ComponentTags []
      FilteredBones []
      FilteredSockets [JNT_MoonSocket]
      MeshUserParameter {parameter {name=None, typeDefHandle {registeredTypeIndex=1}, varData []}}
      SamplingRegions []
     }}
    Socket Sampling Mode=::NewEnumerator5
   )
   &ShapeLocation(
    Additional Quaternion Rotation=true
    Additional Yaw / Pitch / Roll=<0 -43.0827 -90>
    MaskSpawnGroup=false
    Non Uniform Scale=<0.1 0.1 1.5>
    Offset=&AddVectorToPosition(Position=$Output.SkeletalMeshLocation.SampledPositionWithoutOffset, Vector=<0 0 60>)
    Offset Mode=::NewEnumerator0
    Override Seed=false
    Ring / Disc Distribution Mode:ENiagaraSphereDistributionMode=::NewEnumerator1
    Ring Radius=500
    Rotation Angle Type:ENiagara_AngleInput=::NewEnumerator0
    Rotation Mode=::NewEnumerator3
    Rotation Quaternion=$Particles.SkeletalMeshLocation.SampledRotation
    Shape Primitive=::NewEnumerator4
    U Position=&Add_Float(
     A=&SelectFloatFromArray(
      Direct Array Index=&ReturnParticleIDAttributes()
      Float Selection Array=DataInterface() {propertyValues {FloatData [-0.06,0.58,0.33,0.78,1.0809]}}
     )
     B=0
    )
    UseOwnerScale=true
   )
   &SolveForcesAndVelocity(Clamp Velocity=false, Limit Acceleration=false)
   /SetVariables(
    Particles.InnerStarCO=5
    Particles.InnerStarSize:Vector3f=&VectorFromFloat(Value=&FloatFromCurve(
      FloatCurve=DataInterface() {propertyValues {Curve {
         defaultValue=0
         keys [
          {arriveTangent=9.0725, interpMode=RCIM_Cubic, leaveTangent=9.0725, tangentMode=RCTM_User, time=0, value=0}
          {arriveTangent=-2.5249, interpMode=RCIM_Cubic, leaveTangent=-2.5249, tangentMode=RCTM_User, time=0.1408, value=0.9847}
          {arriveTangent=-0.0405, interpMode=RCIM_Cubic, leaveTangent=-0.0405, tangentMode=RCTM_User, time=1, value=0}
         ]
        }}}
      Scale Curve=&RandomRangeFloat(Minimum=0.3, Maximum=0.4)
     ))
   ) {isSetParameters=true, module=SetVariables_#3}
   &FN_Light_Attributes(Color=<rgba 0.7968 0.8407 1.2 1>, Create Color Attribute=true, Create Radius Attribute=true, Radius=60)
  ]
  rendererClasses [&NiagaraLightRendererProperties]
  renderers [{rendererClass=&NiagaraLightRendererProperties, rendererIndex=0}]
 }
 NebulaBG {rendererClasses [], renderers []}
 ShootingStarHead {
  emitterUpdateScript [
   &EmitterState(Life Cycle Mode=::NewEnumerator1, Loop Duration=1, UseLoopDelay=false)
   &SpawnRate(Spawn Probability=0.6, SpawnRate=3, Use Spawn Probability=true)
  ]
  particleSpawnScript [
   &InitializeParticle(
    Color=<rgba 1.1892 2.5818 5 1>
    Lifetime Max=0.8
    Lifetime Min=0.6
    Lifetime Mode=::NewEnumerator1
    Mesh Scale Mode=::NewEnumerator1
    Mesh Uniform Scale Max=0.6
    Mesh Uniform Scale Min=0.2
    Sprite Size Mode=::NewEnumerator3
    Uniform Sprite Size=37.0655
    UsePositionOffset=true
    Write Material Random=true
   )
   &InitialMeshOrientation(Mesh Orientation Mode=::NewEnumerator1, Rotation=<0.25 0.25 -0.16>, Use Rotation Vector=true)
   &ShapeLocation(
    MaskSpawnGroup=false
    Non Uniform Scale=<0.1 0.1 1.5>
    Offset=&Lerp_Vector(
     A=<0 80 125>
     B=<0 -80 125>
     Alpha=&FloatFromCurve(
      FloatCurve=DataInterface() {propertyValues {Curve {keys [<at 0 1 RCIM_Cubic>,<at 1 0 RCIM_Cubic>]}}}
      CurveIndex=&RemapRange(
       Input Value To Remap=&MakeFloatFromVector(
        Vector=&ReturnCameraVector(Camera Vectors=::NewEnumerator1, Coordinate Space=::Simulation)
        Channel=::NewEnumerator0
       )
       Input Min=-0.9
       Input Max=0.9
       Output Min=0
       Output Max=1
      )
     )
    )
    Offset Mode=::NewEnumerator0
    Override Seed=false
    Rotation Angle=93.0804
    Rotation Angle Type:ENiagara_AngleInput=::NewEnumerator0
    Rotation Axis=<0 0 0>
    Rotation Mode=::NewEnumerator1
    Shape Origin=$Engine.Emitter.SimulationPosition
    Sphere Radius=32.8682
    UseOwnerScale=true
   )
   &AddVelocity(Velocity=<-350 120 0>)
  ]
  particleUpdateScript [
   &ParticleState()
   &ScaleMeshSize(Scale Factor=&VectorFromFloat(Value=&FloatFromCurve(FloatCurve=DataInterface() {propertyValues {Curve {
         defaultValue=0
         keys [
          {arriveTangent=9.0725, interpMode=RCIM_Cubic, leaveTangent=9.0725, tangentMode=RCTM_User, time=0, value=0}
          {arriveTangent=-2.5249, interpMode=RCIM_Cubic, leaveTangent=-2.5249, tangentMode=RCTM_User, time=0.1408, value=0.9847}
          {arriveTangent=-0.0405, interpMode=RCIM_Cubic, leaveTangent=-0.0405, tangentMode=RCTM_User, time=1, value=0}
         ]
        }}})))
   &GravityForce(Gravity=<0 0 -4765.9717>)
   &MeshRotationForce(Rotation=&RandomRangeVector(Minimum=<-15 0 0>, Maximum=<-25 0 0>))
   &Drag(
    Drag=&FloatFromCurve(
     FloatCurve=DataInterface() {propertyValues {Curve {keys [<at 0 0>,<at 0.8 1>]}}}
     Scale Curve=&RandomRangeFloat(Minimum=12, Maximum=8.9418)
    )
    Rotational Drag:NiagaraFloat=&RandomRangeFloat(Minimum=2, Maximum=5)
    Use Linear Drag=true
    Use Rotational Drag=true
   )
   &SolveForcesAndVelocity(
    Clamp Velocity=false
    Limit Acceleration=false
    Velocity:Vector3f=&Multiply_VectorByFloat(Vector=&ReturnCameraVector(Camera Vectors=::NewEnumerator3, Coordinate Space=::Simulation), Float=-300)
   )
   &ScaleColor(
    Scale Alpha=&FloatFromCurve(FloatCurve=DataInterface() {propertyValues {Curve {
        defaultValue=0
        keys [
         {arriveTangent=9.0725, interpMode=RCIM_Cubic, leaveTangent=9.0725, tangentMode=RCTM_User, time=0, value=0}
         {arriveTangent=-2.5249, interpMode=RCIM_Cubic, leaveTangent=-2.5249, tangentMode=RCTM_User, time=0.1408, value=0.9847}
         {arriveTangent=-0.0405, interpMode=RCIM_Cubic, leaveTangent=-0.0405, tangentMode=RCTM_User, time=1, value=0}
        ]
       }}})
    Scale RGB=&VectorFromFloat(Value=&FloatFromCurve(FloatCurve=DataInterface() {propertyValues {Curve {
         defaultValue=0
         keys [
          {arriveTangent=9.0725, interpMode=RCIM_Cubic, leaveTangent=9.0725, tangentMode=RCTM_User, time=0, value=0}
          {arriveTangent=-2.5249, interpMode=RCIM_Cubic, leaveTangent=-2.5249, tangentMode=RCTM_User, time=0.1408, value=0.9847}
          {arriveTangent=-0.0405, interpMode=RCIM_Cubic, leaveTangent=-0.0405, tangentMode=RCTM_User, time=1, value=0}
         ]
        }}}))
    ScaleA=true
    ScaleRGB=true
   )
   /SetVariables(
    Particles.GlowColor=&ColorFromCurve(ColorCurve=DataInterface() {propertyValues {
       AlphaCurve {keys [<at 0 1 RCIM_Cubic>,<at 0.5629 0 RCIM_Cubic>]}
       BlueCurve {keys [<at 0 1 RCIM_Cubic>]}
       GreenCurve {keys [<at 0 0.3979 RCIM_Cubic>]}
       RedCurve {keys [<at 0 0.0628 RCIM_Cubic>]}
      }})
    Particles.InnerStarCO=5
    Particles.InnerStarSize:Vector3f=&VectorFromFloat(Value=&FloatFromCurve(
      FloatCurve=DataInterface() {propertyValues {Curve {
         defaultValue=0
         keys [
          {arriveTangent=9.0725, interpMode=RCIM_Cubic, leaveTangent=9.0725, tangentMode=RCTM_User, time=0, value=0}
          {arriveTangent=-2.5249, interpMode=RCIM_Cubic, leaveTangent=-2.5249, tangentMode=RCTM_User, time=0.1408, value=0.9847}
          {arriveTangent=-0.0405, interpMode=RCIM_Cubic, leaveTangent=-0.0405, tangentMode=RCTM_User, time=1, value=0}
         ]
        }}}
      Scale Curve=&RandomRangeFloat(Minimum=0.3, Maximum=0.4)
     ))
   ) {isSetParameters=true, module=SetVariables_#3}
   &FN_Light_Attributes(Color=<rgba 0.5 0.5 0.5 1>, Create Color Attribute=true, Create Radius Attribute=true, Radius=20)
  ]
  rendererClasses [
   &NiagaraMeshRendererProperties
   &NiagaraLightRendererProperties
   &NiagaraSpriteRendererProperties
  ]
  renderers [
   {rendererClass=&NiagaraMeshRendererProperties, rendererIndex=0}
   {rendererClass=&NiagaraLightRendererProperties, rendererIndex=1}
   {rendererClass=&NiagaraSpriteRendererProperties, rendererIndex=2}
  ]
 }
 ShootingStarRibbon {
  emitterUpdateScript [
   &EmitterState()
   &SpawnParticlesFromOtherEmitter(
    Attribute Reader=DataInterface() {propertyValues {EmitterBinding {bindingMode=Other, emitterName=ShootingStarHead}}}
    Spawn Rate=60
    Use Maximum Age=false
    Use Minimum Age=false
   )
  ]
  particleSpawnScript [
   &InitializeParticle(
    Color=<rgba 3.75 4.6145 5 1>
    Lifetime=1.2
    Ribbon Width=&RandomRangeFloat(Minimum=2, Maximum=5)
    Ribbon Width Mode=::NewEnumerator1
    UsePositionOffset=false
    Write Material Random=true
   )
   &SampleParticlesFromOtherEmitter(
    Age Sampling:ENiagara_DirectReadSamplingMode=::NewEnumerator0
    Applied Alpha Scale=0.3
    Applied Position Offset:Vector3f=<-2 0 0>
    Apply Sampled RGB=false
    Color Sampling=::NewEnumerator0
    Particle ID Sampling=::NewEnumerator1
   )
  ]
  particleUpdateScript [
   &ParticleState()
   &SolveForcesAndVelocity(Clamp Velocity=false, Limit Acceleration=false)
   &ScaleRibbonWidth(Ribbon Width Scale=&FloatFromCurve(FloatCurve=DataInterface() {propertyValues {Curve {
        defaultValue=1
        keys [
         {arriveTangent=-1.9314, interpMode=RCIM_Cubic, leaveTangent=-1.9314, tangentMode=RCTM_User, time=0, value=1}
         {interpMode=RCIM_Cubic, tangentMode=RCTM_User, time=1, value=0}
        ]
       }}}))
   &ScaleColor(
    Scale Alpha=&FloatFromCurve(FloatCurve=DataInterface() {propertyValues {Curve {
        defaultValue=1
        keys [
         {tangentMode=RCTM_User, time=0, value=1}
         {arriveTangent=-0.0405, interpMode=RCIM_Cubic, leaveTangent=-0.0405, tangentMode=RCTM_User, time=0.7009, value=1}
         {arriveTangent=-8.0089, interpMode=RCIM_Cubic, leaveTangent=-8.0089, tangentMode=RCTM_User, time=1, value=0}
        ]
       }}})
    Scale RGB=&VectorFromFloat(Value=&FloatFromCurve(FloatCurve=DataInterface() {propertyValues {Curve {
         defaultValue=0
         keys [
          {arriveTangent=9.0725, interpMode=RCIM_Cubic, leaveTangent=9.0725, tangentMode=RCTM_User, time=0, value=0}
          {arriveTangent=-2.5249, interpMode=RCIM_Cubic, leaveTangent=-2.5249, tangentMode=RCTM_User, time=0.1408, value=0.9847}
          {arriveTangent=-0.0405, interpMode=RCIM_Cubic, leaveTangent=-0.0405, tangentMode=RCTM_User, time=1, value=0}
         ]
        }}}))
    ScaleA=true
    ScaleRGB=true
   )
  ]
  rendererClasses [&NiagaraRibbonRendererProperties]
  renderers [{rendererClass=&NiagaraRibbonRendererProperties}]
 }
 SmokeSprites {
  emitterUpdateScript [
   /SetVariables(Emitter.RandomPicker:NiagaraInt32=&RandomRangeInt(Minimum=1, Maximum=2)) {isSetParameters=true}
   &self:InterpolateSpeedThresholdEmitter_02(
    Define Initial Value=false
    Rate Of Change=5
    Target Float Value:NiagaraFloat=&MakeCustomFloatFromBool(
     Bool=&SetBoolByFloatComparison(
      A=&Add_Float(
       A=&Add_Float(
        A=&Abs_Float(Float:NiagaraFloat=&MakeFloatFromVector(Vector=$Engine.Owner.Velocity, Channel=::NewEnumerator0))
        B=&Abs_Float(Float:NiagaraFloat=&MakeFloatFromVector(Vector=$Engine.Owner.Velocity, Channel=::NewEnumerator2))
       )
       B=&Abs_Float(Float:NiagaraFloat=&MakeFloatFromVector(Vector=$Engine.Owner.Velocity, Channel=::NewEnumerator1))
      )
      B=46.278
     )
     True Float=1
     False Float=0
    )
   ) {module=InterpolateSpeedThresholdEmitter_02}
   &EmitterState(Life Cycle Mode=::NewEnumerator1, Loop Delay=0.1, Loop Duration=0.7, Recalculate Duration Each Loop=true, UseLoopDelay=false)
   &SpawnBurst_Instantaneous(
    Spawn Count=&MakeIntFromFloat(Float=&Lerp_Float(A=0, B=25, Alpha=$Emitter.InterpolateSpeedThresholdEmitter_02.InterpolateOverTime_Float.Moving Average))
    Use Spawn Probability=false
    UseLoopCountLimit=false
   )
   &SpawnRate(
    SpawnGroup=1
    SpawnRate=&Lerp_Float(
     A=&MakeCustomFloatFromBool(Bool=$User.bIsFrontEndPreview, True Float=12, False Float=17)
     B=0
     Alpha=$Emitter.InterpolateSpeedThresholdEmitter_02.InterpolateOverTime_Float.Moving Average
    )
    Use Spawn Probability=false
   )
  ]
  particleSpawnScript [
   &PartitionParticles(Number Of Partitions=&MergeHelper_IntSwitch(0=$Emitter.RandomPicker, 1=0, 2=?, 3=?, 4=?, 5=?, Index=&SpawnGroup(), Num=::NewEnumerator0))
   &InitializeParticle(
    Color Mode=::NewEnumerator0
    Lifetime=&MergeHelper_FloatSwitch(
     0=&RandomRangeFloat(Minimum=2.5, Maximum=3.5)
     1=&RandomRangeFloat(Minimum=2, Maximum=3)
     2=?
     3=?
     Index=&SpawnGroup()
     Num=::NewEnumerator0
    )
    Position Offset=&MergeHelper_VectorSwitch(
     0=&MergeHelper_VectorSwitch(
      0=&RandomRangeVector(Minimum=<30 200 30>, Maximum=<90 160 30>)
      1=&RandomRangeVector(Minimum=<-30 100 30>, Maximum=<-90 130 30>)
      2=?
      3=?
      Index=$Particles.PartitionParticles.Partition
      Num=::NewEnumerator0
     )
     1=<0 0 30>
     2=?
     3=?
     Index=&SpawnGroup()
     Num=::NewEnumerator0
    )
    Position Offset Coordinate Space=::Local
    Sprite Rotation Mode=::NewEnumerator2
    Sprite Size Mode=::NewEnumerator1
    Uniform Sprite Size Max=71.856
    Uniform Sprite Size Min=58.2737
    UsePositionOffset=true
    Write Material Random=true
   ) {module=InitializeParticle001}
   &ShapeLocation(
    MaskSpawnGroup=true
    Non Uniform Scale=<1.7318 1 1>
    Offset Mode=::NewEnumerator0
    Override Seed=false
    Rotation Angle Type:ENiagara_AngleInput=::NewEnumerator2
    Rotation Mode=::NewEnumerator2
    Sphere Radius=27.2915
    UseOwnerScale=true
   )
   &ShapeLocation(
    Box Size=<130 130 10>
    MaskSpawnGroup=true
    Override Seed=false
    Shape Primitive=::NewEnumerator2
    Spawn Group Mask=1
    Surface Only=false
    UseOwnerScale=true
   ) {module=ShapeLocation001}
  ]
  particleUpdateScript [
   &ParticleState()
   &SpriteRotationRate(Rotation Rate=&RandomRangeFloat(Minimum=-0.376, Maximum=0.2991))
   &ScaleColor(
    Scale Alpha=&Multiply_Float(
     A=&FloatFromCurve(
      FloatCurve=DataInterface() {propertyValues {Curve {keys [<at 0 0 RCIM_Cubic>, {interpMode=RCIM_Cubic, tangentMode=RCTM_User, time=0.5148, value=0.9647}, <at 1 0 RCIM_Cubic>]}}}
      Scale Curve=&MakeCustomFloatFromBool(Bool=$User.bIsFrontEndPreview, True Float=0.5, False Float=1)
     )
     B=&FloatFromCurve(
      FloatCurve=DataInterface() {propertyValues {Curve {keys [<at -0.0011 1 RCIM_Cubic>, {interpMode=RCIM_Cubic, tangentMode=RCTM_User, time=0.5148, value=0}]}}}
      CurveIndex=&NormalizeDistanceRange(
       Start Position:NiagaraPosition=$Engine.Emitter.SimulationPosition
       End Position:NiagaraPosition=$Particles.Position
       Distance:NiagaraFloat=211.7561
       Invert Normalized Range:NiagaraBool=false
      )
     )
    )
    Scale RGB=&MakeVectorFromLinearColorRGB(Color=&FNTimeOfDayColor(
      Night Color=<rgba 0.4791 0.6431 1 0.6174>
      Morning Color=<rgba 0.6407 0.7252 1 1>
      Day Color=&ColorFromCurve(
       ColorCurve=DataInterface() {propertyValues {
         AlphaCurve {keys [<at 0 1 RCIM_Cubic>]}
         BlueCurve {keys [<at 0 1 RCIM_Cubic>,<at 0.4986 1>]}
         GreenCurve {keys [<at 0 0.7106 RCIM_Cubic>,<at 0.4986 0.4285>]}
         RedCurve {keys [<at 0 0.7319 RCIM_Cubic>,<at 0.4986 0.3255>]}
        }}
       CurveIndex=&NormalizeDistanceRange(
        Start Position:NiagaraPosition=$Engine.Emitter.SimulationPosition
        End Position:NiagaraPosition=$Particles.Position
        Distance:NiagaraFloat=211.7561
        Invert Normalized Range:NiagaraBool=false
       )
      )
      Evening Color=<rgba 0.6407 0.7252 1 1>
     ))
    ScaleA=true
    ScaleRGB=true
   )
   &InheritSourceMovement(
    Applied Position Scale=&MakeCustomVectorFromBool(
     Bool=&SetBoolByFloatComparison(A=&VectorLength(Vector=$User.OwnerLocalVelocity), B=350)
     True Vector=<0.9 0.9 1>
     False Vector=<0.6 0.6 0.6>
    )
    ApplyForce=false
    ApplyPosition=true
    ApplyPositionFalloff=false
   )
   &SolveForcesAndVelocity(Clamp Velocity=false, Limit Acceleration=false)
   &CameraOffset(Camera Offset Amount=5)
   &DynamicMaterialParameters(
    Index 0 Param 1=&Add_Float(
     A=&RandomRangeFloat(Minimum=0, Maximum=5)
     B=&FloatFromCurve(
      FloatCurve=DataInterface() {propertyValues {Curve {keys [
          {arriveTangent=0.1897, interpMode=RCIM_Cubic, leaveTangent=0.1897, tangentMode=RCTM_User, time=0, value=1}
          {arriveTangent=0.0529, interpMode=RCIM_Cubic, leaveTangent=0.2021, tangentMode=RCTM_User, time=1.0036, value=1.1706}
         ]}}}
      CurveIndex=&Add_Float(A=&Multiply_Float(A=$Particles.Initial.MaterialRandom, B=0.5), B=$Particles.NormalizedAge)
     )
    )
    Index 0 Param 2=&FloatFromCurve(FloatCurve=DataInterface() {propertyValues {Curve {keys [
         {arriveTangent=0.0453, interpMode=RCIM_Cubic, tangentMode=RCTM_Break, time=0, value=-1}
         {arriveTangent=0.0604, interpMode=RCIM_Cubic, leaveTangent=0.0604, tangentMode=RCTM_User, time=1, value=-1}
        ]}}})
    Index 0 Param 3=&FloatFromCurve(
     FloatCurve=DataInterface() {propertyValues {Curve {keys [{arriveTangent=-3.6642, interpMode=RCIM_Cubic, leaveTangent=-3.6642, tangentMode=RCTM_User, time=0, value=1}]}}}
     Scale Curve=&Multiply_Float(
      A=&Multiply_Float(A=0.002, B=&FNTimeOfDayScalar(Night Scale=1, Morning Scale=0, Day Scale=0, Evening Scale=0))
      B=&Lerp_Float(A=1, B=0, Alpha=&ClampFloat(Float=$Particles.DistFromCandle))
     )
    )
    Index 0 Param 4=&FloatFromCurve(FloatCurve=DataInterface() {propertyValues {Curve {keys [<at 0.3 0>,<at 0.4 1>,<at 1 0>]}}}, Scale Curve=0.9)
    Index 1 Param 1=&FloatFromCurve(
     FloatCurve=DataInterface() {propertyValues {Curve {keys [
         {arriveTangent=2.8574, interpMode=RCIM_Cubic, leaveTangent=2.8574, tangentMode=RCTM_User, time=0, value=1}
         {arriveTangent=4.2257, interpMode=RCIM_Cubic, leaveTangent=4.2257, tangentMode=RCTM_User, time=0.25, value=1}
        ]}}}
     Scale Curve=&RandomRangeFloat(Minimum=0.5, Maximum=1)
    )
    Index 1 Param 2=&FloatFromCurve(
     FloatCurve=DataInterface() {propertyValues {Curve {keys [<at 0 0.2>, {leaveTangent=0.9066, time=1, value=0.3}]}}}
     Scale Curve=&RandomRangeFloat(Minimum=0.5, Maximum=0.6)
    )
    Index 1 Param 3=&FloatFromCurve(
     FloatCurve=DataInterface() {propertyValues {Curve {keys [
         {arriveTangent=2.0526, interpMode=RCIM_Cubic, leaveTangent=2.0526, tangentMode=RCTM_User, time=0, value=0}
         {interpMode=RCIM_Cubic, tangentMode=RCTM_User, time=1, value=1}
        ]}}}
     Scale Curve=0
    )
    Index 1 Param 4=&FloatFromCurve(
     FloatCurve=DataInterface() {propertyValues {Curve {keys [{arriveTangent=-0.3224, interpMode=RCIM_Cubic, leaveTangent=-0.3224, tangentMode=RCTM_User, time=0, value=1}]}}}
     Scale Curve=0
    )
    Index 2 Param 1=&FloatFromCurve(FloatCurve=DataInterface() {propertyValues {Curve {keys [
         {interpMode=RCIM_Cubic, leaveTangent=4.2485, tangentMode=RCTM_Break, time=0, value=0}
         {arriveTangent=1.0548, interpMode=RCIM_Cubic, leaveTangent=1.0548, tangentMode=RCTM_User, time=0.17, value=0.5715}
         <at 1 1>
        ]}}})
    Index 2 Param 2=&FloatFromCurve(
     FloatCurve=DataInterface() {propertyValues {Curve {keys [
         {arriveTangent=1.9056, interpMode=RCIM_Cubic, leaveTangent=1.9056, tangentMode=RCTM_User, time=0, value=0.2}
         {arriveTangent=0.0259, interpMode=RCIM_Cubic, leaveTangent=0.0259, tangentMode=RCTM_User, time=1, value=1.4}
        ]}}}
     Scale Curve=&RandomRangeFloat(Minimum=0.5, Maximum=0.7)
    )
    Index 2 Param 3=&FloatFromCurve(FloatCurve=DataInterface() {propertyValues {Curve {keys [
         {interpMode=RCIM_Cubic, leaveTangent=2.3064, tangentMode=RCTM_Break, time=0, value=0}
         {arriveTangent=2, interpMode=RCIM_Cubic, leaveTangent=2, time=0.2, value=1.4}
         {interpMode=RCIM_Cubic, leaveTangent=11.1233, tangentMode=RCTM_Break, time=1, value=2}
        ]}}})
    Index 2 Param 4=&FloatFromCurve(
     FloatCurve=DataInterface() {propertyValues {Curve {keys [
         {arriveTangent=-0.6867, interpMode=RCIM_Cubic, leaveTangent=-0.6867, tangentMode=RCTM_User, time=0, value=1}
         {interpMode=RCIM_Cubic, leaveTangent=-1.2956, tangentMode=RCTM_Break, time=1, value=0.75}
        ]}}}
     Scale Curve=1.333
    )
    Module.Param0WriteEnabled=true
    Module.Param0WriteEnabled1=true
    Module.Param0WriteEnabled2=true
    Module.Param1WriteEnabled=false
    Module.Param1WriteEnabled1=true
    Module.Param1WriteEnabled2=true
    Module.Param2WriteEnabled=true
    Module.Param2WriteEnabled1=true
    Module.Param2WriteEnabled2=true
    Module.Param3WriteEnabled=true
    Module.Param3WriteEnabled1=true
    Module.Param3WriteEnabled2=true
    Write Parameter Index 1=true
    Write Parameter Index 2=true
   )
   &CameraOffset(Camera Offset Amount=7) {module=CameraOffset001}
  ]
  rendererClasses [&NiagaraSpriteRendererProperties]
  renderers [
   {rendererClass=&NiagaraSpriteRendererProperties}
   {rendererClass=&NiagaraSpriteRendererProperties, rendererIndex=1}
  ]
 }
 SparkleStars {rendererClasses [], renderers []}
}
summary {
 systemName=NS_Emote_FavorDockEntry_MoonFX
 emitters [
  {
   emitterName=ShootingStarHead
   rendererClasses [
    &NiagaraMeshRendererProperties
    &NiagaraLightRendererProperties
    &NiagaraSpriteRendererProperties
   ]
  }
  {emitterName=ShootingStarRibbon, rendererClasses [&NiagaraRibbonRendererProperties]}
  {emitterName=NebulaBG, rendererClasses []}
  {emitterName=SmokeSprites, rendererClasses [&NiagaraSpriteRendererProperties]}
  {emitterName=MeshStars, rendererClasses []}
  {emitterName=MoonLight, rendererClasses [&NiagaraLightRendererProperties]}
  {emitterName=SparkleStars, rendererClasses []}
 ]
 userVariables [
  User.bIsFrontEnd {type=NiagaraBool}
  User.bIsFrontEndPreview {type=NiagaraBool}
  User.OwnerLocalVelocity {type=Vector3f, description="", defaultValue {struct=Vector3f, value=<0 0 0>}}
 ]
}
